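; Weapon Blueprint for the anti-vehicle mounted turret that plays its firing effects on the host vehicle.
; On host vehicle setup it caches the Fort Athena vehicle and checks its ArmoredBattleBus gameplay tag to drive a bus flag, with a leftover debug print.
; When firing, it spawns a pooled muzzle flash on the vehicle's skeletal mesh, choosing the socket by that flag, and shakes the local viewing player's camera.
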(event EventOnPlayWeaponFireFX
  (params (bPersistentFire :type Boolean) (bSecondaryFire :type Boolean))
  (bind _instigator (Variables|Actor|GetInstigator))
  (bind _returnvalue (Pawn|GetLocalViewingPlayerController _instigator))
  (bind _athenavehicle (Variables|Default|GetAthenaVehicle))
  ;[1 1 1] MuzzleFlash FX Fire Camera Shake
  (Utilities|IsValid _athenavehicle
    (:Is%20Valid
      (Effects|SpawnFXSystemAttached :FXSystemAsset (Variables|Default|GetMuzzleFlashFX) :AttachToComponent (Class|FortVehicleInterface|GetSkeletalMeshComponent _athenavehicle) :AttachPointName (select (Variables|Default|GetIsArmoredBattleBus) "MuzzleLeft_f" "MuzzleLeft") :Scale "1.000000,1.000000,1.000000" :LocationType "KeepRelativeOffset" :bAutoDestroy true :PoolingMethod "AutoRelease" :bAutoActivate true)
      (Utilities|IsValid _returnvalue
        (:Is%20Valid
          (CameraShakes|StartCameraShake :self (Variables|/Script/Engine.PlayerController|GetPlayerCameraManager _returnvalue) :ShakeClass (Variables|Default|GetFireCameraShake) :Scale 1.0 :PlaySpace "CameraLocal"))))))

(event EventMuzzleFlashFX
  (params (Persistent%20Fire :type Boolean) (OffHand :type Boolean)))

(event Setup|EventOnHostVehicleSetup
  ;[1 1 1] Check if Host vehicle is not battle bus
  (|CallParentFunction)
  (bind _asfort_athena_vehicle (Utilities|Casting|CastToFortAthenaVehicle (Utilities|ToObject(Interface) (Vehicle|GetHostVehicle))))
  (Variables|Default|SetAthenaVehicle _asfort_athena_vehicle)
  (if (GameplayTags|HasMatchingGameplayTag _asfort_athena_vehicle "(TagName=\"Athena.Vehicle.ArmoredBattleBus\")")
    (Development|PrintString "IM THE BATTLE BUS" true true "(R=0.000000,G=0.660000,B=1.000000,A=1.000000)" 15.0)
    (Variables|Default|SetIsArmoredBattleBus true)))

(fn ConstructionScript
  (|CallParentFunction))
