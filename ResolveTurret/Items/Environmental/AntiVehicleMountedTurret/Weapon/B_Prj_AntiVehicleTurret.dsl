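; Rocket projectile for the mounted anti-vehicle turret, handling its fuse, collision exceptions and impact effects.
; On authority, vehicle tags decide whether it ignores the firer's vehicle and passengers.
; Touch either stops it or, on "NoStick" components, doubles its velocity.
; Stopping or fuse expiry schedules an impact event that plays water or standard explosion effects, sound and a camera-shake cue, with scorch decals outside water.
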(event EventBeginPlay
  (bind _vehicletags (Variables|Default|GetVehicleTags))
  ;[1 1 1] EVENT -> Begin Play
  (Utilities|Time|SetTimerbyFunctionName self "FuseTimerMax" 25.0)
  (Collision|SetIgnoreInstigatorCollision true)
  (bind _asfort_player_pawn (Utilities|Casting|CastToFortPlayerPawn (Variables|Actor|GetInstigator)))
  (bind _output_get (Variables|ProjectileRocket|SetPlayerPawn _asfort_player_pawn))
  (Networking|SwitchHasAuthority
    (:Authority
      (bind _returnvalue (Vehicle|GetVehicleActor (Vehicle|GetVehicleInterface _output_get) :!class /Script/FortniteGame.FortVehicleInterface))
      (Collision|IgnoreActorwhenMoving (Variables|Collision|GetCapsuleComponent) _returnvalue (select (GameplayTags|HasAnyTags (Class|FortVehicleInterface|GetVehicleTags (Vehicle|GetVehicleInterface _output_get)) _vehicletags) true false))
      (for _array_element (Vehicle|GetAllPassengers (Vehicle|GetVehicleInterface _output_get))
        (Collision|IgnoreActorwhenMoving (Variables|Collision|GetCapsuleComponent) _array_element (select (GameplayTags|HasAnyTags (Class|FortVehicleInterface|GetVehicleTags (Vehicle|GetVehicleInterface _output_get)) _vehicletags) true false))))))

(event EventOnStop
  (params (Hit :type Hit%20Result%20Structure%20%28by%20ref%29))
  (bind (_bblockinghit _binitialoverlap _time _distance _location _impactpoint _normal _impactnormal _physmat) (Collision|BreakHitResult Hit))
  ;[1 1 1] Impact
  (Variables|Effects|SetStopLocation (+ _impactpoint (* _impactnormal (Variables|Effects|GetExplosion%20Impact%20Offset))))
  (Variables|Effects|SetStopRotZ (Math|Conversions|RotationFromXVector _impactnormal))
  (Utilities|IsValid _physmat
    (:Is%20Valid
      (Variables|ProjectileRocket|SetSurfaceType (Variables|/Script/PhysicsCore.PhysicalMaterial|GetSurfaceType _physmat)))
    (:Is%20Not%20Valid
      _))
  (if (not (Delivery|IsBeingKilled))
    (Utilities|Time|SetTimerbyEvent (EventDispatchers|CreateEvent :!event Custom|Projectile_Impact) (Variables|Default|GetTime))
    (Fortnite|ShowSoundIndicatorLocal :Location (Transformation|GetWorldLocation (Variables|Projectile_Rocket|GetMesh)) :MaxAudibleDistance 5e+03 :IndicatorType "Gunshot" :Instigator self :Tint "(R=0.000000,G=0.000000,B=0.000000,A=0.000000)" :AttenuationMaxRangeMultiplier 1.0 :AttenuationIntensityMultiplier 1.0)))

(event Custom|FuseTimerMax
  (bind _mesh (Variables|Projectile_Rocket|GetMesh))
  (bind _returnvalue (Transformation|GetWorldLocation _mesh))
  ;[1 1 1] Impact
  (Variables|Effects|SetTimerMaxReached? true)
  (Variables|Effects|SetStopLocation _returnvalue)
  (if (not (Delivery|IsBeingKilled))
    (Utilities|Time|SetTimerbyEvent (EventDispatchers|CreateEvent :!event Custom|Projectile_Impact) (Variables|Default|GetTime))
    (Fortnite|ShowSoundIndicatorLocal :Location _returnvalue :MaxAudibleDistance 5e+03 :IndicatorType "Gunshot" :Instigator self :Tint "(R=0.000000,G=0.000000,B=0.000000,A=0.000000)" :AttenuationMaxRangeMultiplier 1.0 :AttenuationIntensityMultiplier 1.0)))

(event Custom|Projectile_Impact
  (bind _stoprotz (Variables|Effects|GetStopRotZ))
  (bind _stoplocation (Variables|Effects|GetStopLocation))
  ;[1 1 1] Impact
  ;;[1 1 1] EVENT -> On Pojectile Impact
  (Networking|CanExecuteCosmeticEvents
    (:True
      ;;[1 1 1] SingleLineCheckForWorldObjectsSurfaceType
      (if (or (Utilities|Enum|Equal(Enum) (Variables|ProjectileRocket|GetSurfaceType) "SurfaceType8") (Water|IsinWater (Variables|Water|GetWaterInteractionComponent)))
        (bind _returnvalue (Effects|Components|ParticleSystem|SpawnFXAtLocation (Variables|FX|GetFX_Water_Explosion_Asset) _stoplocation _stoprotz))
        (Rendering|SetRenderCustomDepth _returnvalue true)
        (Audio|PlaySoundatLocation (Variables|Sounds|GetProjectile_Explosion_Water_Sound) (Variables|Effects|GetStopLocation))
        (else
          (bind _returnvalue_2 (Effects|Components|ParticleSystem|SpawnFXAtLocation (Variables|FX|GetFX_Explosion_Asset) _stoplocation _stoprotz))
          (Rendering|SetRenderCustomDepth _returnvalue_2 true)
          (Audio|PlaySoundatLocation (Variables|Sounds|GetProjectile_Explosion_Sound) (Variables|Effects|GetStopLocation))
          (Utilities|SpawnDecalSimple "12.000000,35.000000,35.000000" "12.000000,40.000000,40.000000" (Variables|Effects|GetStopLocation) (Variables|Effects|GetStopRotZ) 3.0 3.5 "/Game/Weapons/Materials/MD_DecalSimple_Master_01.MD_DecalSimple_Master_01" "/Game/Effects/Fort_Effects/Gameplay/SmokeDustExplosions/Textures/T_Decal_Scorch_Hole.T_Decal_Scorch_Hole" "(R=14.000000,G=1.914246,B=0.000000,A=1.000000)" "(R=1.000000,G=0.236675,B=0.000000,A=1.000000)"
            (:Out
              _))))
      ;;[1 1 1] Shake that camera
      (GameplayCue|ExecuteGameplayCue%28Burst%29Local self (Variables|Default|GetFeedbackCue) (Ability|GameplayCue|MakeGameplayCueParameters 0.0)) ; Cam shake + controller vibration
    )
    (:False
      (CallFunction|HideonDeath))))

(event EventDestroyed
  ;[1 1 1] Impact
  ;;[1 1 1] EVENT -> Destroyed
  (Networking|CanExecuteCosmeticEvents
    (:True
      (Components|Activation|SetActive (Variables|Effects|GetRocketTrailPS) false false))
    (:False
      (CallFunction|HideonDeath))))

(event Custom|HideOnDeath
  ;[1 1 1] Impact
  ;;[1 1 1] EVENT -> Destroyed
  (Development|SetHiddeninGame (Variables|Projectile_Rocket|GetMesh) true false))

(event EventOnTouched
  (params (OtherActor :type Actor%20Object%20Reference) (OtherComp :type Primitive%20Component%20Object%20Reference) (HitResult :type Hit%20Result%20Structure%20%28by%20ref%29) (bIsOverlap :type Boolean))
  (bind (_bblockinghit _binitialoverlap _time _distance _location _impactpoint _normal _impactnormal _physmat) (Collision|BreakHitResult HitResult))
  (bind _surfacetype (Variables|/Script/PhysicsCore.PhysicalMaterial|GetSurfaceType _physmat))
  (bind (_bblockinghit_2 _binitialoverlap_2 _time_2 _distance_2 _location_2 _impactpoint_2 _normal_2 _impactnormal_2 _physmat_2 _hitactor _hitcomponent) (Collision|BreakHitResult HitResult))
  ;[1 1 1] Bounce off "NoBounce"
  (Utilities|IsValid _physmat
    (:Is%20Valid
      (if (or (Utilities|Enum|Equal(Enum) _surfacetype "SurfaceType8") (not bIsOverlap))
        (if (Components|ComponentHasTag _hitcomponent "NoStick")
          (Variables|/Script/Engine.MovementComponent|SetVelocity (* (Variables|/Script/Engine.MovementComponent|GetVelocity (Variables|Movement|GetProjectileMovementComponent)) 2.0) (Variables|Movement|GetProjectileMovementComponent))
          (else
            (Variables|Projectile|SetbStopSimulatingOnHit true)
            (Variables|ProjectileRocket|SetSurfaceType _surfacetype)))))))

(fn SetMaterialParameters
  (params (LEFT :type Vector) (RIGHT :type Vector) (Index :type EDualWeaponHand%20Enum :default "LEFT"))
  (bind _mesh (Variables|Projectile_Rocket|GetMesh))
  (Rendering|Material|SetVectorParameterValueonMaterials _mesh "BulletOrigin" (+ (* (Transformation|GetActorForwardVector) 5e+01) (Utilities|Select :Index Index :LEFT LEFT :RIGHT RIGHT)))
  (Rendering|Material|SetScalarParameterValueonMaterials _mesh "StartTime" (Utilities|Time|GetGameTimeinSeconds)))

(fn ConstructionScript
  (Effects|Components|ParticleSystem|SetTemplate (Variables|Default|GetRibbon_Trail_PSC) (Variables|Effects|GetRibbon_Trail_PS))
  (Effects|Components|ParticleSystem|SetTemplate (Variables|Default|GetExplosion_Generic_PSC) (Variables|Effects|GetExplosion_Generic_PS))
  (Effects|Components|ParticleSystem|SetTemplate (Variables|Default|GetExplosion_Flesh_Damage_PSC) (Variables|Effects|GetExplosion_Flesh_Damage_PS)))
